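# Animation Blueprint for the M_LRG_AvocadoSeal large male player body, a custom character part that follows the parent mesh's pose.
# The copied parent pose feeds the body's Control Rig, then a per-bone layered blend adds DefaultSlot animation on forearm and wrist plate bones.
# While an emote plays, each update reads the owning player anim instance's emote_BuiltIn curve to set that blend weight.
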AnimBlueprint /BRCosmetics/Characters/Player/Male/Large/Bodies/M_LRG_AvocadoSeal/Meshes/M_LRG_AvocadoSeal_AnimBP
target_skeleton=&/M_LRG_AvocadoSeal_Skeleton
graphs [
 AnimGraph {nodes [
   n1 = Control Rig(Source=Copy Pose From Mesh() {bUseAttachedParent=true}) {
    alphaBoolBlend {blendOption=Linear}
    controlRigAssetReference {blueprintRigClass=&/M_LRG_AvocadoSeal_CtrlRig.M_LRG_AvocadoSeal_CtrlRig_C}
   }
   n2 = Layered blend per bone(BasePose=@n1, BlendPoses_0=Slot 'DefaultSlot'(), BlendWeights_0=Get Is Playing Custom Emote() -> (Is Playing Custom Emote:double)) {layerSetup [{branchFilters [
       {boneName=forearm_plate_01_jnt}
       {boneName=forearm_plate_02_jnt}
       {boneName=forearm_plate_03_jnt}
       {boneName=wrist_plate_01_jnt}
       {boneName=wrist_plate_02_jnt}
      ]}]}
   Output Pose(Result=@n2) {layerGroup=None}
  ]}
 EventGraph {nodes [
   Event BlueprintUpdateAnimation()
   > Branch(Condition=Get bIsPlayingEmote())
   > Set Is Playing Custom Emote(Is Playing Custom Emote:double=GetCurveValue(
     self=GetOwnerFortPlayerAnimInstance!(self:Custom Character Part Anim Instance Object Reference) -> (ReturnValue:Fort Player Anim Instance Object Reference)
     CurveName=emote_BuiltIn
    )) -> (Output_Get:double)
  ]}
]
variables [Is Playing Custom Emote {type=number}]
