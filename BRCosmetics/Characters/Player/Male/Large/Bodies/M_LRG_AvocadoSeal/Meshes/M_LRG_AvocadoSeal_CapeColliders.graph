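# Physics asset of collision shapes for the large male AvocadoSeal player body, named as the colliders for its cape.
# Stacked capsules wrap the pelvis and spine, with two extra long capsules on spine_02, plus capsules on the neck and both clavicles and a sphere on the head.
# Arm, leg and foot bodies carry no shapes, and no constraints exist, so it provides collision volumes rather than a jointed ragdoll.
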PhysicsAsset /BRCosmetics/Characters/Player/Male/Large/Bodies/M_LRG_AvocadoSeal/Meshes/M_LRG_AvocadoSeal_CapeColliders
bodies {
 calf_l []
 calf_r []
 clavicle_l [{shapeType=Capsule, length=9, radius=12, center=<9.3 -1 1>, rotation=<rot 75 0 0>}]
 clavicle_r [{shapeType=Capsule, length=9, radius=12, center=<-9.3 1 -1>, rotation=<rot 75 0 0>}]
 foot_l []
 foot_r []
 head [{shapeType=Sphere, radius=17.3125, center=<11.028 -0.174 0>}]
 lowerarm_l []
 lowerarm_r []
 neck_02 [{shapeType=Capsule, length=13.4375, radius=12, center=<-0.2642 -0.3408 0>, rotation=<rot 90 -156.3706 -170.6486>}]
 pelvis [{shapeType=Capsule, length=10, radius=18, center=<-8.0188 -2.9947 0>}]
 spine_01 [{shapeType=Capsule, length=10, radius=18, center=<-7.3531 -0.7559 0>}]
 spine_02 [
  {shapeType=Capsule, length=10, radius=18, center=<-1.5 0.6597 0>}
  {shapeType=Capsule, length=29.3956, radius=11.5807, center=<2.5024 0.4992 3.9825>, rotation=<rot 90 0.0049 0.0049>}
  {shapeType=Capsule, length=29.875, radius=11.1013, center=<2.4704 0.4991 -3.9825>, rotation=<rot 90 -0.0396 -0.0396>}
 ]
 spine_03 [{shapeType=Capsule, length=10, radius=19, center=<0 0.3147 0>}]
 spine_04 [{shapeType=Capsule, length=12, radius=19, center=<0.907 -0.4007 0>}]
 spine_05 [{shapeType=Capsule, length=12, radius=19, center=<-2.2795 -2.1684 0>}]
 thigh_l []
 thigh_r []
 upperarm_l []
 upperarm_r []
}
constraints []
